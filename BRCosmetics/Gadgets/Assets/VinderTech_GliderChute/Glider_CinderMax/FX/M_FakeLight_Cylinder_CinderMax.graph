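# Additive unlit material for Niagara mesh particles that fakes light from the CinderMax glider inside a cylinder volume.
# It reconstructs scene position behind the translucency in instance space for radial and height falloff, scaled by particle alpha.
# Scene normals weight it toward surfaces facing against the glider's up axis; Low and Medium quality zero this term, removing the effect.
# Only backfaces render, with depth testing off, avoiding camera clipping.
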Material /BRCosmetics/Gadgets/Assets/VinderTech_GliderChute/Glider_CinderMax/FX/M_FakeLight_Cylinder_CinderMax
graph {
 properties {
  bDisableDepthTest=true
  bUsedWithNiagaraMeshParticles=true
  blendMode=BLEND_Additive
  pixelDepthOffsetMode=PDOM_Legacy
  shadingModel=MSM_Unlit
  twoSided=true
  usedShadingModels=Unlit
 }
 nodes [
  n1 = Transform(None=Constant3Vector() {constant=<rgba 0 0 1 0>}) {desc=Glider Up Axis, transformSourceType=TRANSFORMSOURCE_Instance}
  n2 = Multiply(A=ComponentMask(None=SceneTexture() {sceneTextureId=PPI_WorldNormal}.Color) {b=true, g=true, r=true}) {constB=-1}
  n3 = Power(Base=Saturate(None=DotProduct(A=@n1, B=@n2)), Exp=ScalarParameter() {defaultValue=2, parameterName=Normal Exponent})
  n4 = Constant()
  n5 = QualitySwitch(Default=@n3, Low=@n4, High=@n3, Medium=@n4, Epic=@n3)
  n6 = TransformPosition(None=MaterialFunctionCall() {materialFunction=&WorldPositionBehindTranslucency}.Result) {transformSourceType=TRANSFORMPOSSOURCE_World, transformType=TRANSFORMPOSSOURCE_Instance}
  n7 = ComponentMask(None=@n6) {g=true, r=true}
  n8 = Multiply(A=Subtract(A=MaterialFunctionCall(Vector 2=@n7) {materialFunction=/Engine/Functions/Engine_MaterialFunctions02/Utility/VectorLength}.V2 Length) {constB=96, desc=Radius of the mesh}) {constB=-1}
  n9 = ScalarParameter() {defaultValue=33, parameterName=Fade Falloff Distance}
  n10 = ComponentMask(None=@n6) {b=true}
  n11 = Saturate(None=ConstantBiasScale(None=Divide(A=@n10) {constB=96, desc=Mesh half height * scale}) {desc=Mesh pivot is at the center of height})
  n12 = Multiply(A=Saturate(None=Divide(A=@n8, B=@n9)), B=Frac(None=OneMinus(None=@n11)))
  n13 = TwoSidedSign()
  Multiply(
   A=Multiply(A=ParticleColor().A, B=Multiply(A=@n5, B=@n12))
   B=OneMinus(None=Saturate(None=@n13)) {desc=Made the material two sided and disabled depth testing to avoid camera clipping}
  )
 ]
}
